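; Parent fire volume Blueprint that applies a Curie fire interaction to actors overlapping it.
; On authority, it waits for playlist data and destroys itself if the fire volumes scalable float is disabled.
; Begin overlap skips player pawns in vehicles and already tracked actors, then starts the interaction at a scalable-float magnitude.
; Interaction handles are stored in a per-actor map, so end overlap can end the matching interaction and remove its entry.
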(event EventBeginPlay
  (Networking|SwitchHasAuthority
    (:Authority
      (Class|FortAsyncActionCurrentPlaylistReady|PlaylistDataReadyAsync
        (:OnReady
          (if (not (ScalableFloat|GetValueasBool (Variables|Default|GetFireVolumesEnabled)))
            (Actor|DestroyActor)))))))

(event Collision|EventActorBeginOverlap
  (params (OtherActor :type Actor%20Object%20Reference))
  (CallFunction|ActorEnteredFireVolume OtherActor))

(event Collision|EventActorEndOverlap
  (params (OtherActor :type Actor%20Object%20Reference))
  (CallFunction|ActorExitedFireVolume OtherActor))

(fn ActorEnteredFireVolume
  (params (Actor :type Actor%20Object%20Reference))
  (bind _asfort_player_pawn (Utilities|Casting|CastToFortPlayerPawn Actor))
  (if (not (or (Vehicle|IsinVehicle _asfort_player_pawn) (Utilities|Map|Contains (Variables|Default|GetInteractHandleMap) Actor)))
    (bind _returnvalue (Curie|ExecuteBeginFireInteract Actor self Actor (ScalableFloat|GetValueatLevel (Variables|Default|GetCureFireMagnitude)) (Collision|MakeHitResult false false 0.0 0.0 (Transformation|GetActorLocation Actor) (Transformation|GetActorLocation Actor) "0,0,1" "0,0,1" 0 Actor)))
    (Utilities|Map|Add (Variables|Default|GetInteractHandleMap) Actor _returnvalue)))

(fn ActorExitedFireVolume
  (params (Actor :type Actor%20Object%20Reference))
  (bind _interacthandlemap (Variables|Default|GetInteractHandleMap))
  (bind (_value _returnvalue) (Utilities|Map|Find _interacthandlemap Actor))
  (if _returnvalue
    (Curie|ExecuteEndFireInteract Actor _value)
    (Utilities|Map|Remove (Variables|Default|GetInteractHandleMap) Actor)))

(fn ConstructionScript)
